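# Third-person cue for a character jumping in water in Athena.
# An IsEnemy branch picks one of two mirrored mixes, playing enemy or unset cases louder than non-enemies.
# Each layers random Outlander jump foley, swim-jump splashes, and Soldier water-landing steps that are pitched up by stacked modulators and faded in by an envelope.
# Jump attenuation applies, and owner-limited concurrency allows one instance, preventing new overlapping plays.
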SoundCue /Game/Athena/Sounds/Character/Footsteps/3P/Water/Athena_Character_Jump_Water_3P_Cue
root=Branch_1
nodes [
 {id=Branch_1, children [Mixer_2,Mixer_4,Mixer_2], properties {boolParameterName=IsEnemy}}
 {
  id=Enveloper_1
  children [Modulator_1]
  properties {
   pitchCurve {editorCurveData {keys [{time=0}]}}
   volumeCurve {editorCurveData {keys [<at -0.0002 0.0075>,<at 0.1817 0.0667>,<at 0.3 0.5626>]}}
  }
 }
 {
  id=Enveloper_2
  children [Modulator_5]
  properties {
   pitchCurve {editorCurveData {keys [{time=0}]}}
   volumeCurve {editorCurveData {keys [<at -0.0002 0.0075>,<at 0.1817 0.0667>,<at 0.3 0.5626>]}}
  }
 }
 {id=Mixer_1, children [Random_1,Modulator_2,Random_3], properties {inputVolume [1,1,1]}}
 {id=Mixer_2, children [Mixer_1], properties {inputVolume [0.65]}}
 {id=Mixer_3, children [Random_7,Modulator_4,Random_8], properties {inputVolume [1,1,1]}}
 {id=Mixer_4, children [Mixer_3], properties {inputVolume [0.35]}}
 {id=Modulator_1, children [Random_2], properties {pitchMax=1.4, pitchMin=1.2}}
 {id=Modulator_2, children [Enveloper_1], properties {pitchMax=1.3, pitchMin=1.1, volumeMax=1, volumeMin=0.7}}
 {id=Modulator_4, children [Enveloper_2], properties {pitchMax=1.3, pitchMin=1.1, volumeMax=1, volumeMin=0.7}}
 {id=Modulator_5, children [Random_0], properties {pitchMax=1.4, pitchMin=1.2}}
 {
  id=Random_0
  children [WavePlayer_10, WavePlayer_11, WavePlayer_12, WavePlayer_13, WavePlayer_14, WavePlayer_15]
  properties {bShouldExcludeFromBranchCulling=true, weights [1,1,1,1,1,1]}
 }
 {
  id=Random_1
  children [WavePlayer_0,WavePlayer_7,WavePlayer_8,WavePlayer_9]
  properties {bShouldExcludeFromBranchCulling=true, weights [1,1,1,1]}
 }
 {
  id=Random_2
  children [WavePlayer_1,WavePlayer_2,WavePlayer_3,WavePlayer_4,WavePlayer_5,WavePlayer_6]
  properties {bShouldExcludeFromBranchCulling=true, weights [1,1,1,1,1,1]}
 }
 {id=Random_3, children [WavePlayer_16, WavePlayer_17], properties {bShouldExcludeFromBranchCulling=true, weights [1,1]}}
 {
  id=Random_7
  children [WavePlayer_21, WavePlayer_22, WavePlayer_23, WavePlayer_26]
  properties {bShouldExcludeFromBranchCulling=true, weights [1,1,1,1]}
 }
 {id=Random_8, children [WavePlayer_27, WavePlayer_30], properties {bShouldExcludeFromBranchCulling=true, weights [1,1]}}
 {id=WavePlayer_0, properties {soundWaveAssetPtr=/Game/Sounds/Fortress_Foley/Outlander/Outlander_Jump_004}}
 {id=WavePlayer_1, properties {soundWaveAssetPtr=/Game/Sounds/Fort_Step_Sounds/Soldier/Soldier_Step_Land_Water_01}}
 {id=WavePlayer_10, properties {soundWaveAssetPtr=&Soldier_Step_Land_Water_01}}
 {id=WavePlayer_11, properties {soundWaveAssetPtr=/Game/Sounds/Fort_Step_Sounds/Soldier/Soldier_Step_Land_Water_02}}
 {id=WavePlayer_12, properties {soundWaveAssetPtr=/Game/Sounds/Fort_Step_Sounds/Soldier/Soldier_Step_Land_Water_03}}
 {id=WavePlayer_13, properties {soundWaveAssetPtr=/Game/Sounds/Fort_Step_Sounds/Soldier/Soldier_Step_Land_Water_04}}
 {id=WavePlayer_14, properties {soundWaveAssetPtr=/Game/Sounds/Fort_Step_Sounds/Soldier/Soldier_Step_Land_Water_05}}
 {id=WavePlayer_15, properties {soundWaveAssetPtr=/Game/Sounds/Fort_Step_Sounds/Soldier/Soldier_Step_Land_Water_06}}
 {id=WavePlayer_16, properties {soundWaveAssetPtr=/Game/Athena/Sounds/Water/Interaction/Water_Player_SwimJump_1P_01}}
 {id=WavePlayer_17, properties {soundWaveAssetPtr=/Game/Athena/Sounds/Water/Interaction/Water_Player_SwimJump_1P_02}}
 {id=WavePlayer_2, properties {soundWaveAssetPtr=&Soldier_Step_Land_Water_02}}
 {id=WavePlayer_21, properties {soundWaveAssetPtr=&Outlander_Jump_004}}
 {id=WavePlayer_22, properties {soundWaveAssetPtr=/Game/Sounds/Fortress_Foley/Outlander/Outlander_Jump_001}}
 {id=WavePlayer_23, properties {soundWaveAssetPtr=/Game/Sounds/Fortress_Foley/Outlander/Outlander_Jump_002}}
 {id=WavePlayer_26, properties {soundWaveAssetPtr=/Game/Sounds/Fortress_Foley/Outlander/Outlander_Jump_003}}
 {id=WavePlayer_27, properties {soundWaveAssetPtr=&Water_Player_SwimJump_1P_01}}
 {id=WavePlayer_3, properties {soundWaveAssetPtr=&Soldier_Step_Land_Water_03}}
 {id=WavePlayer_30, properties {soundWaveAssetPtr=&Water_Player_SwimJump_1P_02}}
 {id=WavePlayer_4, properties {soundWaveAssetPtr=&Soldier_Step_Land_Water_04}}
 {id=WavePlayer_5, properties {soundWaveAssetPtr=&Soldier_Step_Land_Water_05}}
 {id=WavePlayer_6, properties {soundWaveAssetPtr=&Soldier_Step_Land_Water_06}}
 {id=WavePlayer_7, properties {soundWaveAssetPtr=&Outlander_Jump_001}}
 {id=WavePlayer_8, properties {soundWaveAssetPtr=&Outlander_Jump_002}}
 {id=WavePlayer_9, properties {soundWaveAssetPtr=&Outlander_Jump_003}}
]
properties {
 attenuationSettings=/Game/Sounds/Attenuation/Fort_Player_Jump_3P_Athena_Att
 bOverrideConcurrency=true
 maxDistance=3584
 soundClassObject=&Fort_Player_Foley_3P_Enemy_Athena
 volumeMultiplier=1
 concurrencyOverrides {bLimitToOwner=true, maxCount=1, resolutionRule=PreventNew}
}
